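# Animation Blueprint for the Alien Flora built-in male pickaxe, which changes petal color when the owning player takes damage.
# A watch on the DamageReceived cosmetic stat raises a target value in clamped steps, then resets it after a retriggerable delay.
# Each update tracks equip flags and eases the current value toward that target into the MaterialCurveA material parameter.
# The AnimGraph only constrains lowerarm_r to root on the input pose.
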AnimBlueprint /BRCosmetics/Weapons/FORT_Melee/Pickaxe_Alien_Flora_Built_In_Male/Meshes/Alien_Flora_Built_In_Male_Axe_AnimBP
target_skeleton=&/Alien_Flora_Built_In_Male_Axe_Skeleton
graphs [
 AnimGraph {nodes [
   "Transform (Modify) Bone - Bone: lowerarm_twist_01_r"(Alpha=1, Translation="0.000,1.2,-0.000", Rotation="0,0,0", Scale="1,1,1") {
    translationMode=BMM_Additive
    translationSpace=BCS_BoneSpace
    alphaBoolBlend {blendOption=Linear}
    boneToModify {boneName=lowerarm_twist_01_r}
    translation {y=1.2}
   }
   AnimGraphNode_SubInput_0 = Input Pose()
   n1 = "Constraint - lowerarm_r (P)"(ComponentPose=Local To Component(LocalPose=@AnimGraphNode_SubInput_0), Alpha=1, ConstraintWeights_0=1) {
    alphaBoolBlend {blendOption=Linear}
    boneToModify {boneName=lowerarm_r}
    constraintSetup [{offsetOption=None, targetBone {boneName=root}}]
    constraintWeights [1]
   }
   Output Pose(Result=Component To Local(ComponentPose=@n1)) {layerGroup=None}
  ]}
 EventGraph {nodes [
   n1 = Event OnCharacterCustomizationLoaded()
   n2 = Cast To FortPlayerPawn(execute=[@n1.then, @n12.Completed], Object=TryGetFortPawn())
   > n3 = WatchCosmeticStat!(
    execute:Exec
    Pawn:Fort Player Pawn Object Reference=@n2
    StatTag:Fort Stat Manager Tag Structure=(TagName="StatManager.DamageReceived")
    DefaultValue:int=0
    bBroadcastInitialValue:bool=true
    InitialValueBroadcastDelay:float=-1
    ChangeValueBroadcastDelay:float=-1
    bConsumeWaitingNotifications:bool=true
   ) -> (then:Exec, OnStatChanged:Exec, StatName:Name, StatValue:int) {is=title K2Node_AsyncAction}
   Branch(execute=@n3.OnStatChanged, Condition="integer > integer"(A=@n3.StatValue, B=0) {is=title K2Node_CallFunction})
   n4 = Get PetelsTargetColor() -> (PetelsTargetColor:double)
   > Set PetelsTargetColor(PetelsTargetColor:double="Clamp (Float)"(Value=float + float(A=@n4, B=0.334) {is=title K2Node_CommutativeAssociativeBinaryOperator}, Min=0, Max=1)) -> (Output_Get:double)
   > RetriggerableDelay(Duration=6)
   > Set PetelsTargetColor(PetelsTargetColor=0)
   n5 = Event BlueprintUpdateAnimation()
   > n6 = Sequence()
   > n7 = Cast To FortWeaponPickaxeAthena!(execute, Object=GetOwningActor()) -> (then, CastFailed, AsFort Weapon Pickaxe Athena:Fort Weapon Pickaxe Athena Object Reference) {is=title K2Node_DynamicCast}
   > n8 = Set IsEquipped(IsEquipped:bool=IsEquipped!(self:Fort Weapon Object Reference=@n7) -> (ReturnValue:bool)) -> (Output_Get:bool)
   > n9 = Branch(Condition=@n8)
   > n10 = Branch(Condition=Get IsAlreadyEquipped() -> (IsAlreadyEquipped:bool))
   > Set IsInitialEquipped(IsInitialEquipped:bool=false) -> (Output_Get:bool)
   Set IsAlreadyEquipped(execute=@n10.else, IsAlreadyEquipped:bool=true) -> (Output_Get:bool)
   Set IsAlreadyEquipped(execute=@n9.else, IsAlreadyEquipped=false)
   > Set IsInitialEquipped(IsInitialEquipped=true)
   n11 = Set PetelsCurrentColor(
    execute=@n6.then_1
    PetelsCurrentColor:double=FInterpTo(Current=Get PetelsCurrentColor() -> (PetelsCurrentColor:double), Target=Get PetelsTargetColor(), DeltaTime=@n5.DeltaTimeX, InterpSpeed=2)
   ) -> (Output_Get:double)
   > SetScalarParameterValueOnMaterials!(execute, self:Mesh Component Object Reference=GetPartSkeletalMeshComponent(), ParameterName:Name=MaterialCurveA, ParameterValue:float=@n11) -> (then)
   Event OnWeaponChanged()
   > n12 = Do Once()
   TryGetPawnOwner()
   Append!(A="IsAlreadyEquipped = ", B, C:String="\r\n", D:String="IsInitialEquipped = ", E:String) -> (ReturnValue)
   n13 = Event BlueprintProcessAnimInputEvent()
   OR Boolean(
    A=IsNamed(self=@n13.AnimInputEvent, InEventName=aie.PlayerVitality.HealthDecreased)
    B=IsNamed(self=@n13.AnimInputEvent, InEventName=aie.PlayerVitality.ShieldsDecreased)
   )
  ]}
]
variables [
 IsAlreadyEquipped {type=boolean}
 IsEquipped {type=boolean}
 IsInitialEquipped {type=boolean}
 PetelsCurrentColor {type=number}
 PetelsTargetColor {type=number}
 Skel_Mesh {type=SkeletalMeshComponent}
]
