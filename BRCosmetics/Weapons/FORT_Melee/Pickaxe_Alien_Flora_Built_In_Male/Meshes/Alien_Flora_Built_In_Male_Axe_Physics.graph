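# Physics asset for the Alien Flora Built-In Male pickaxe mesh, a melee weapon cosmetic.
# It holds a single body on the lowerarm_r bone, shaped as one long capsule offset mostly along the bone's negative X axis.
# It defines no constraints, so there is no jointed chain to simulate.
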PhysicsAsset /BRCosmetics/Weapons/FORT_Melee/Pickaxe_Alien_Flora_Built_In_Male/Meshes/Alien_Flora_Built_In_Male_Axe_Physics
bodies {lowerarm_r [{shapeType=Capsule, length=85.866, radius=13.9762, center=<-57.7666 -6.3433 0.9669>, rotation=<rot 0 91.2458 -89.9685>}]}
constraints []
